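; BP_Sprite_Laser_Component is an actor component, named for the item version of a laser sprite in the WhisperRemedy SweetSprite content.
; On BeginPlay it gets its owning actor, checks that it is valid, casts it to FortPlayerPawn and stores the result in OwnerPlayerPawn.
; The Tick event is empty, and the graph shows no laser behaviour beyond caching the owner pawn.
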(event EventBeginPlay
  (bind _returnvalue (Components|GetOwner))
  (Utilities|IsValid _returnvalue
    (:Is%20Valid
      (bind _asfort_player_pawn (Utilities|Casting|CastToFortPlayerPawn _returnvalue))
      (Variables|Default|SetOwnerPlayerPawn _asfort_player_pawn))))

(event EventTick
  (params (DeltaSeconds :type Float%20%28single-precision%29)))
