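; Backpack Blueprint for the WhisperRemedySweet Sprite, a creature worn on the player's back that animates in reaction to its wearer.
; BeginPlay attaches the cape to the spine, flags backpack idle, and briefly delays visibility so animation updates first.
; Gameplay-tag watchers on the instigator drive the skydiving anim flag and unequip montage.
; Events play reaction montages, some randomized, with crouch suppressed when DBNO; local players' backpacks skip auto-calculated significance.
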(event EventBeginPlay
  (bind _returnvalue (Game|GetInstigator))
  (bind _spriteanimbp (Variables|Default|GetSpriteAnimBP))
  (bind _spriteanimbp_2 (Variables|Default|GetSpriteAnimBP))
  (bind _assprite_whisper_remedy_sweet_creature_skeleton_anim_blueprint (Utilities|Casting|CastToSprite_WhisperRemedySweet_Creature_Skeleton_AnimBlueprint (Components|SkeletalMesh|GetAnimInstance (Variables|Default|GetSkeletalMeshComponentBudgeted))))
  (Transformation|AttachComponentToComponent (Variables|Default|GetSprite_Cape_Creature_Scale) (Variables|Default|GetSkeletalMeshComponentBudgeted) "spine_1" "KeepRelative" "KeepRelative" "KeepRelative" true)
  (bind _output_get (Variables|Default|SetSpriteAnimBP _assprite_whisper_remedy_sweet_creature_skeleton_anim_blueprint))
  (Variables|/Script/SpritesRuntime.SpriteCreatureAnimInstance|SetbIsBackpack_Idle true _output_get)
  (Utilities|FlowControl|Sequence
    (:then_0
      (bind (_asynctaskproxy _tagcount) (Ability|Tasks|WaitGameplayTagCountChangedOnActor (Game|GetInstigator) "(TagName=\"Athena.Player.Sprite.Equipped\")")
        (:then
          (bind (_asynctaskproxy_2 _tagcount_2) (Ability|Tasks|WaitGameplayTagCountChangedOnActor _returnvalue "(TagName=\"MovementMode.Skydiving\")")
            (:then
              (bind (_asynctaskproxy_3 _tagcount_3) (Ability|Tasks|WaitGameplayTagCountChangedOnActor _returnvalue "(TagName=\"MovementMode.Parachuting\")")
                (:TagCountChanged
                  (if (> _tagcount_3 0)
                    (Variables|/Script/SpritesRuntime.SpriteCreatureAnimInstance|SetbIsPawnSkydiving true _spriteanimbp_2)
                    (else
                      (Utilities|FlowControl|Delay 0.2)
                      (Variables|/Script/SpritesRuntime.SpriteCreatureAnimInstance|SetbIsPawnSkydiving false _spriteanimbp_2))))))
            (:TagCountChanged
              (if (> _tagcount_2 0)
                (Variables|/Script/SpritesRuntime.SpriteCreatureAnimInstance|SetbIsPawnSkydiving true _spriteanimbp)
                (else
                  (Variables|/Script/SpritesRuntime.SpriteCreatureAnimInstance|SetbIsPawnSkydiving false _spriteanimbp))))))
        (:TagCountChanged
          (if (< _tagcount 1)
            (Animation|Montage|MontagePlay _output_get "/Sprites_WhisperRemedySweetSprite/Assets/Animation/UnEquip_Sprite_Water_M.UnEquip_Sprite_Water_M" 1.0 "MontageLength" 0.0 true)))))
    (:then_1
      ; TDiA RM - delaying vis to allow anim to update on Pickup
      (Utilities|FlowControl|Delay 0.05)
      (Rendering|SetVisibility (Variables|Default|GetSkeletalMeshComponentBudgeted) true false))))

(event EventPlayAgitatedMontage
  (if (Math|Random|RandomBool)
    (Animation|Montage|PlayMontage (Variables|Default|GetSkeletalMeshComponentBudgeted) "/Sprites_WhisperRemedySweetSprite/Assets/Animation/Backpack_Agitated_Water_M.Backpack_Agitated_Water_M"
      (:OnCompleted
        (Development|PrintString "Agitated Montage")))))

(event EventPlayCelebrateMontage
  (Animation|Montage|PlayMontage (Variables|Default|GetSkeletalMeshComponentBudgeted) (Utilities|Select :Index (Math|Random|RandomIntegerinRange 0 2) :Option%200 "/Sprites_WhisperRemedySweetSprite/Assets/Animation/Backpack_CelebrateA_Water_M.Backpack_CelebrateA_Water_M" :Option%201 "/Sprites_WhisperRemedySweetSprite/Assets/Animation/Backpack_CelebrateB_Water_M.Backpack_CelebrateB_Water_M" :Option%202 "/Sprites_WhisperRemedySweetSprite/Assets/Animation/Backpack_CelebrateC_Water_M.Backpack_CelebrateC_Water_M")
    (:OnCompleted
      (Development|PrintString "Celebrate Montage"))))

(event EventBackpackAttached
  (params (PlayerPawn :type Fort%20Player%20Pawn%20Object%20Reference))
  ; Clear auto calc significance on locally controlled player sprite backpack so significance is always 1.0
  (if (Pawn|IsLocallyControlled PlayerPawn)
    (Class|SpritesUtilityLibrary|SpritesSetAutoCalculateSignificanceonSkeletalMeshComponentBudgeted (Variables|Default|GetSkeletalMeshComponentBudgeted))))

(event EventPlayCrouchMontage
  (bind _returnvalue (GameplayTags|HasMatchingGameplayTag(Message) (Game|GetInstigator) "(TagName=\"Gameplay.Status.DBNO\")"))
  (if _returnvalue
    _
    (elif (Math|Random|RandomBool)
      (Animation|Montage|PlayMontage (Variables|Default|GetSkeletalMeshComponentBudgeted) "/Sprites_WhisperRemedySweetSprite/Assets/Animation/Backpack_Crouch_Water_M.Backpack_Crouch_Water_M"
        (:then
          _)
        (:OnCompleted
          _))
      (Development|PrintString "Crouch Montage"))))

(event EventPlayCryMontage
  (Animation|Montage|PlayMontage (Variables|Default|GetSkeletalMeshComponentBudgeted) "/Sprites_WhisperRemedySweetSprite/Assets/Animation/Cry_Water_Sprite_Water_M.Cry_Water_Sprite_Water_M"))

(fn ConstructionScript)
