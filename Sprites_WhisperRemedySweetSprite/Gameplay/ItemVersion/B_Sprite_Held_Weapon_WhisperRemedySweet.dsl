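; Held weapon Blueprint for the Whisper Remedy Sweet Sprite, a laser beam weapon that overheats.
; The replicated laser state's OnRep runs timers that raise or lower overheat, or a cooldown on reaching max, tuned by ScalableFloats.
; Overheat changes notify listeners and, off authority, feed beam audio.
; Gameplay-tag async tasks track vehicle entry and targeting to toggle the idle Niagara effect and optionally apply a vehicle-aim GameplayEffect.
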(event EventBeginPlay
  ;[1 1 1] Lifecycle
  (|CallParentFunction)
  (Networking|SwitchHasAuthority
    (:Authority
      (Components|DestroyComponent (Variables|Default|GetBeamAudioController)))))

(event FortWeapon|EventOnLocalUnEquip
  ;[1 1 1] Lifecycle
  (Rendering|SetVisibility (Variables|Visual|GetWeaponMesh) false false))

(event Custom|SetUpVehicleException
  ;[1 1 1] Vehicle Aiming
  (Utilities|FlowControl|Sequence
    (:then_0
      (Utilities|IsValid (Variables|Default|GetInVehicleAsyncTask)
        (:Is%20Not%20Valid
          (bind _asynctaskproxy (Ability|Tasks|WaitGameplayTagAddToActor (Variables|Default|GetOwningPawn) "(TagName=\"Gameplay.Action.Player.AthenaInVehicle\")")
            (:then
              (Variables|Default|SetInVehicleAsyncTask _asynctaskproxy))
            (:Added
              (Rendering|SetVisibility (Variables|Default|GetNS_SuperSprite_Idle) false false)
              (Development|PrintString "SET WEAPON VISIBILITY OFF !!!")
              (CallFunction|SetUpAiminginVehicleException))))))
    (:then_1
      (Utilities|IsValid (Variables|Default|GetExitVehicleAsyncTask)
        (:Is%20Not%20Valid
          (bind _asynctaskproxy_2 (Ability|Tasks|WaitGameplayTagRemoveFromActor (Variables|Default|GetOwningPawn) "(TagName=\"Gameplay.Action.Player.AthenaInVehicle\")")
            (:then
              (Variables|Default|SetExitVehicleAsyncTask _asynctaskproxy_2))
            (:Removed
              (Rendering|SetVisibility (Variables|Default|GetNS_SuperSprite_Idle) true false)
              (Development|PrintString "SET WEAPON VISIBILITY ON !!!"))))))))

(event FortWeapon|EventOnEquippedAndAttached
  ;[1 1 1] Lifecycle
  (|CallParentFunction)
  (Utilities|FlowControl|Sequence
    (:then_0
      (bind _assprite_whisper_remedy_sweet_creature_skeleton_anim_blueprint (Utilities|Casting|CastToSprite_WhisperRemedySweet_Creature_Skeleton_AnimBlueprint (Components|SkeletalMesh|GetAnimInstance (Variables|Visual|GetWeaponMesh))))
      (Variables|/Script/SpritesRuntime.SpriteCreatureAnimInstance|SetbIsEquipped_Idle true _assprite_whisper_remedy_sweet_creature_skeleton_anim_blueprint))
    (:then_1
      (Utilities|FlowControl|Delay 0.1)
      (Rendering|SetVisibility (Variables|Visual|GetWeaponMesh) true false))
    (:then_2
      (Variables|Default|SetOwningPawn (Actor|GetOwner)))
    (:then_3
      (CallFunction|SetUpVehicleException))))

(event Custom|SetUpAimingInVehicleException
  ;[1 1 1] Vehicle Aiming
  (Utilities|FlowControl|Sequence
    (:then_0
      (Utilities|IsValid (Variables|Default|GetStartTargetingAsyncTask)
        (:Is%20Not%20Valid
          (bind _asynctaskproxy (Ability|Tasks|WaitGameplayTagAddToActor (Variables|Default|GetOwningPawn) "(TagName=\"Gameplay.Status.IsTargeting\")")
            (:then
              (Variables|Default|SetStartTargetingAsyncTask _asynctaskproxy))
            (:Added
              (Rendering|SetVisibility (Variables|Default|GetNS_SuperSprite_Idle) true false)
              (Development|PrintString "TARGETING IN VEHICLE !!!")
              (bind _asfort_pawn (Utilities|Casting|CastToFortPawn (Variables|Default|GetOwningPawn)))
              (if (and (ScalableFloat|GetValueasBool (Variables|Default|GetbShouldUseVehicleAim)) (== (Weapon|GetCurrentWeapon _asfort_pawn) self))
                (bind _returnvalue (GameplayEffects|ApplyGameplayEffectToSelf (Ability|GetAbilitySystemComponent (Variables|Default|GetOwningPawn)) "/Sprites_WhisperRemedySweetSprite/Gameplay/ItemVersion/GE_WhisperRemedy_Sweet_VehicleTargeting.GE_WhisperRemedy_Sweet_VehicleTargeting_C" 0.0 (Class|AthenaFunctionLibrary|MakeSimpleFortEffectContext (Variables|Default|GetOwningPawn) (Variables|Default|GetOwningPawn))))
                (Variables|Default|SetVehicleAimGEHandle _returnvalue)))))))
    (:then_1
      (Utilities|IsValid (Variables|Default|GetStopTargetingAsyncTask)
        (:Is%20Not%20Valid
          (bind _asynctaskproxy_2 (Ability|Tasks|WaitGameplayTagRemoveFromActor (Variables|Default|GetOwningPawn) "(TagName=\"Gameplay.Status.IsTargeting\")")
            (:then
              (Variables|Default|SetStopTargetingAsyncTask _asynctaskproxy_2))
            (:Removed
              (Rendering|SetVisibility (Variables|Default|GetNS_SuperSprite_Idle) false false)
              (Development|PrintString "STOP TARGETING IN VEHICLE !!!")
              (if (ScalableFloat|GetValueasBool (Variables|Default|GetbShouldUseVehicleAim))
                (GameplayEffects|RemoveActiveEffectswithTags (Ability|GetAbilitySystemComponent (Variables|Default|GetOwningPawn)) "(GameplayTags=((TagName=\"Ability.WhisperRemedy.Sweet.VehicleTargeting\")))")))))))))

(event FortWeapon|EventOnUnEquippedAndDetached
  ;[1 1 1] Lifecycle
  (|CallParentFunction)
  (Utilities|Time|ClearandInvalidateTimerbyHandle (Variables|Overheat|Transient|GetIncreaseOverheat_TimerHandle)))

(event Custom|BeginIncreasingOverheat
  ;[1 1 1] Increasing Overheat
  (Utilities|Time|ClearandInvalidateTimerbyHandle (Variables|Overheat|Transient|GetDecreaseOverheat_TimerHandle))
  (bind _returnvalue (Utilities|Time|SetTimerbyEvent (delegate Custom|IncreaseOverheat) (ScalableFloat|GetValueatLevel (Variables|Overheat|ScalableFloat|GetIncrementSpeed)) true))
  (Variables|Overheat|Transient|SetIncreaseOverheat_TimerHandle _returnvalue))

(event Custom|IncreaseOverheat
  (bind _incrementamount (Variables|Overheat|ScalableFloat|GetIncrementAmount))
  (bind _returnvalue (ScalableFloat|GetValueatLevel _incrementamount))
  (bind _currentoverheat (Variables|Overheat|Transient|GetCurrentOverheat))
  (bind _returnvalue_2 (+ _currentoverheat _returnvalue))
  ;[1 1 1] Increasing Overheat
  (if (>= _returnvalue_2 (ScalableFloat|GetValueatLevel (Variables|Overheat|ScalableFloat|GetOverheatMax)))
    (CallFunction|SetOverheat (ScalableFloat|GetValueatLevel (Variables|Overheat|ScalableFloat|GetOverheatMax)))
    (Networking|SwitchHasAuthority
      (:Authority
        (Variables|Default|SetCurrentLaserState "NewEnumerator2"))
      (:Remote
        _))
    (Utilities|Time|ClearandInvalidateTimerbyHandle (Variables|Overheat|Transient|GetIncreaseOverheat_TimerHandle))
    (else
      (bind _overheat (CallFunction|SetOverheat _returnvalue_2))
      (Networking|SwitchHasAuthority
        (:Remote
          (CallFunction|UpdateBeamAudioOverheat _overheat))))))

(event Custom|BeginDecreasingOverheat
  ;[1 1 1] Decreasing Overheat
  (if (> (Variables|Overheat|Transient|GetCurrentOverheat) 0)
    (Utilities|Time|ClearandInvalidateTimerbyHandle (Variables|Overheat|Transient|GetIncreaseOverheat_TimerHandle))
    (Utilities|FlowControl|Delay (ScalableFloat|GetValueatLevel (Variables|Overheat|ScalableFloat|GetDecrementDelay)))
    (if (Utilities|Enum|Equal(Enum) (Variables|Default|GetCurrentLaserState) "NewEnumerator1")
      (bind _returnvalue (Utilities|Time|SetTimerbyEvent (delegate Custom|DecreaseOverheat) (ScalableFloat|GetValueatLevel (Variables|Overheat|ScalableFloat|GetDecrementSpeed)) true))
      (Variables|Overheat|Transient|SetDecreaseOverheat_TimerHandle _returnvalue))))

(event Custom|DecreaseOverheat
  (bind _decrementamount (Variables|Overheat|ScalableFloat|GetDecrementAmount))
  (bind _returnvalue (ScalableFloat|GetValueatLevel _decrementamount))
  (bind _currentoverheat (Variables|Overheat|Transient|GetCurrentOverheat))
  (bind _returnvalue_2 (- _currentoverheat _returnvalue))
  ;[1 1 1] Decreasing Overheat
  (if (<= _returnvalue_2 0)
    (CallFunction|SetOverheat 0.0)
    (Utilities|Time|ClearandInvalidateTimerbyHandle (Variables|Overheat|Transient|GetDecreaseOverheat_TimerHandle))
    (else
      (bind _overheat (CallFunction|SetOverheat _returnvalue_2))
      (Networking|SwitchHasAuthority
        (:Remote
          (CallFunction|UpdateBeamAudioOverheat _overheat))))))

(event Custom|BeginCooldown
  ;[1 1 1] Cooldown
  (Utilities|Time|ClearandInvalidateTimerbyHandle (Variables|Overheat|Transient|GetDecreaseOverheat_TimerHandle))
  (Utilities|Time|ClearandInvalidateTimerbyHandle (Variables|Overheat|Transient|GetIncreaseOverheat_TimerHandle))
  (if (ScalableFloat|GetValueasBool (Variables|Default|GetShouldResetOverheatInstantly))
    (CallFunction|SetOverheat 0.0)
    (else
      _))
  (Utilities|FlowControl|Delay (ScalableFloat|GetValueatLevel (Variables|Overheat|ScalableFloat|GetCooldown)))
  (if (not (ScalableFloat|GetValueasBool (Variables|Default|GetShouldResetOverheatInstantly)))
    (CallFunction|SetOverheat 0.0)
    (else
      _))
  (Networking|SwitchHasAuthority
    (:Authority
      (Variables|Default|SetCurrentLaserState "NewEnumerator1"))))

(event Custom|UpdateBeamAudioOverheat
  (params (Overheat :type Float%20%28double-precision%29))
  ;[1 1 1] Overheat audio
  (if (FortWeapon|IsEquipped)
    (bind _beamaudioactor (BeamAudio|GetBeamAudioActor))
    (Utilities|IsValid _beamaudioactor
      (:Is%20Valid
        (BeamAudio|SetBeamAudioFloatParam _beamaudioactor "Overheat" (Math|Float|NormalizetoRange Overheat 0.0 (ScalableFloat|GetValueatLevel (Variables|Overheat|ScalableFloat|GetOverheatMax))) :!class /Game/Sounds/Fort_Audio_Blueprints/BeamAudio/BPI_BeamAudio_Interface.BPI_BeamAudio_Interface_C)))))

(fn GetBeamAudioActor
  (bind _asbp_beam_audio (Utilities|Casting|CastToBP_BeamAudio (Variables|/Script/Engine.ChildActorComponent|GetChildActor (Variables|Default|GetBeamAudioController)))
    (:then
      (return _asbp_beam_audio))
    (:CastFailed
      (return 0))))

(fn GetBeamOwner
  (return 0 false))

(fn OnRep_CurrentLaserState
  (Default|CallOnLaserStateChanged (Variables|Default|GetCurrentLaserState))
  (switch Utilities|FlowControl|Switch|SwitchonEnum_WhisperRemedy_Sweet_LaserBeamStates (Variables|Default|GetCurrentLaserState)
    (:NewEnumerator0
      (CallFunction|BeginIncreasingOverheat))
    (:NewEnumerator1
      (CallFunction|BeginDecreasingOverheat))
    (:NewEnumerator2
      (CallFunction|BeginCooldown))))

(fn SetOverheat
  (params (InOverheat :type Float%20%28double-precision%29))
  (bind _output_get (Variables|Overheat|Transient|SetCurrentOverheat InOverheat))
  (if (FortWeapon|IsEquipped)
    (Default|CallOnOverheatChanged _output_get))
  (return _output_get))

(fn ShouldShowSoundIndicator
  (params (PlayerController :type Fort%20Player%20Controller%20Object%20Reference))
  ;[1 1 1] FSI Components will auto fire an AVI in AFortWeapon::ShowWeaponFireAudioVisual(). Toggling this prevents that (duplication) and allows the GCNL to handle it instead.
  (return (not (ScalableFloat|GetValueasBool (Variables|Default|GetShouldUseGCAVI)))))

(fn ConstructionScript)
