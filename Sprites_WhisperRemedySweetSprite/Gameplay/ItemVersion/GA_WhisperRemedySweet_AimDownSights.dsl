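; Gameplay ability that handles aim down sights for the item version of the Whisper Remedy Sweet sprite.
; On activation it waits for secondary press and release events: press pushes a secondary laser camera mode and plays the montage unless the RaiseSprite notify tag is present, with debug prints.
; Release pops the camera and stops the montage, as ability end also pops the camera.
; A tag-removal task with no tag specified restarts the montage at its Loop section.
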(event Ability|EventActivateAbility
  (Ability|Tasks|WaitGameplayEvent "(TagName=\"Event.Secondary.Pressed\")"
    (:then
      (Ability|Tasks|WaitGameplayEvent "(TagName=\"Event.Secondary.Released\")"
        (:then
          (Ability|Tasks|WaitGameplayTagRemove :OnlyTriggerOnce false
            (:Removed
              (Ability|Tasks|PlayMontageAndWait "None" (Variables|Default|GetMontage%20to%20Play) 1.0 "Loop"))))
        (:EventReceived
          (Camera|PopCameraMode)
          (Ability|Animation|MontageStop -1.0))))
    (:EventReceived
      (Camera|PushCameraMode "/Sprites_WhisperRemedySweetSprite/Gameplay/ItemVersion/CameraMode_Laser_Secondary_WhisperRemedy_Sweet.CameraMode_Laser_Secondary_WhisperRemedy_Sweet_C")
      (if (GameplayTags|HasMatchingGameplayTag (CallFunction|GetAvatarasFortPlayerPawn) "(TagName=\"Animation.Notify.WhisperRemedy.Sweet.RaiseSprite\")")
        (Development|PrintString "NOT PLAYING MONTAGE !!!")
        (else
          (Ability|Tasks|PlayMontageAndWait "None" (Variables|Default|GetMontage%20to%20Play))
          (Development|PrintString "PLAYING MONTAGE !!!"))))))

(event Ability|EventOnEndAbility
  (params (bWasCancelled :type Boolean))
  (Camera|PopCameraMode))
