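; Gameplay ability that fires the laser beam of the held Whisper Remedy Sweet sprite weapon.
; After a charge-up timer, it starts the weapon's targeted beam component, adds a beam cue and consumes ammo on a looping timer.
; Release, beam timeout, low magazine or a laser state change ends it, but only that state change commits cooldown.
; Ending stops the beam, removes the cue locally on clients, and optionally drops an empty weapon.
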(event Custom|ParentBeginPlay
  ; On Activate Ability
  (bind _outfortplayerpawn (CallFunction|GetAvatarasValidFortPlayerPawn)
    (:ValidFortPlayerPawn
      (bind _output_get (Variables|Default|SetOwnerPawn _outfortplayerpawn))
      (bind _asfort_actor_component_targeted_beam (Utilities|Casting|CastToFortActorComponent_TargetedBeam (Actor|GetComponentbyClass (Weapon|GetCurrentWeapon _output_get) (Variables|Default|GetActorComponentClass)))
        (:then
          (bind _output_get_2 (Variables|Default|SetTargetedBeamComponent _asfort_actor_component_targeted_beam))
          (Utilities|IsValid _output_get_2
            (:Is%20Valid
              (Utilities|Time|SetTimerbyEvent (EventDispatchers|CreateEvent :!event Custom|StartBeam) (Beam|GetBeamChargeUpTime (Variables|Default|GetTargetedBeamComponent)))
              ;; Passive GE is needed for things like granted tags on sim proxies
              (bind _returnvalue (Ability|ApplyGameplayEffectToOwner (Variables|Default|GetPassiveGEClass)))
              (Variables|Default|SetPassiveGEHandle _returnvalue))
            (:Is%20Not%20Valid
              (Fortnite|FortShippingLog self "Could not find valid Targeted beam component! Ending ability." true)
              (Ability|EndAbility))))
        (:CastFailed
          (Fortnite|FortShippingLog self "No TargetedBeam component on weapon! Ending ability." true))))
    (:AvatarCastFailed
      (Fortnite|FortShippingLog self "No valid FortPlayerPawn Avatar! Ending ability." true)))
  (Ability|EndAbility))

(event Ability|EventActivateAbility
  ; On Activate Ability
  (Utilities|FlowControl|Sequence
    (:then_0
      (CallFunction|ParentBeginPlay)
      (bind _asb_sprite_held_weapon_whisper_remedy_sweet (Utilities|Casting|CastToB_Sprite_Held_Weapon_WhisperRemedySweet (Ability|GetCurrentSourceWeapon)))
      (Variables|Default|SetSourceWeapon _asb_sprite_held_weapon_whisper_remedy_sweet)
      (if (GameplayTags|HasMatchingGameplayTag (CallFunction|GetAvatarasFortPlayerPawn) "(TagName=\"Animation.Notify.WhisperRemedy.Sweet.RaiseSprite\")")
        (Development|PrintString "NOT PLAYING MONTAGE !!!")
        (else
          (Ability|Tasks|PlayMontageAndWait "None" (Variables|Default|GetMontage%20to%20Play))
          (Development|PrintString "PLAYING MONTAGE !!!"))))
    (:then_2
      (Ability|Tasks|WaitInputRelease true
        (:then
          (Default|BindEventtoOnLaserStateChanged (Variables|Default|GetSourceWeapon) (delegate Custom|OnLaserStateChanged)))
        (:OnRelease
          (Ability|EndAbility))))
    (:then_3
      (Variables|Default|SetbShouldCommitCooldown false))))

(event Custom|OnLaserStateChanged
  (params (LaserState :type Enum_WhisperRemedy_Sweet_LaserBeamStates%20Enum :default "NewEnumerator0"))
  (if (Utilities|Enum|Equal(Enum) LaserState "NewEnumerator2")
    (Variables|Default|SetbShouldCommitCooldown true)
    (Ability|EndAbility)))

(event Ability|EventOnEndAbility
  (params (bWasCancelled :type Boolean))
  (bind _ownerpawn (Variables|Default|GetOwnerPawn))
  (bind _currentweapon (Variables|/Script/FortniteGame.FortPawn|GetCurrentWeapon _ownerpawn))
  (bind _weaponmesh (Variables|/Script/FortniteGame.FortWeapon|GetWeaponMesh _currentweapon))
  ; On End Ability
  (Default|UnbindEventfromOnLaserStateChanged (Variables|Default|GetSourceWeapon) (delegate Custom|OnLaserStateChanged))
  (Rendering|Material|SetScalarParameterValueonMaterials _weaponmesh "IsInstantRedEye" 0.0)
  (Rendering|Material|SetScalarParameterValueonMaterials _weaponmesh "CoolDownTime" (Utilities|Time|GetGameTimeinSeconds))
  (Utilities|Time|ClearandInvalidateTimerbyHandle (Variables|Default|GetAbilityCostTimer))
  (if (Variables|Default|GetbShouldCommitCooldown)
    (Development|PrintString "COMMITTING COOLDOWN !!!")
    (Ability|CommitAbilityCooldown false)
    (else
      (Variables|/Sprites_WhisperRemedySweetSprite/Gameplay/ItemVersion/B_Sprite_Held_Weapon_WhisperRemedySweet.B_Sprite_Held_Weapon_WhisperRemedySweet_C|SetCurrentLaserState "NewEnumerator1" (Variables|Default|GetSourceWeapon))
      (Development|PrintString "NOT COMMITTING COOLDOWN !!!")))
  (if (Variables|Default|GetbIsEmpty)
    (CallFunction|ParentDoonEndAbility)
    (if (ScalableFloat|GetValueasBool (Variables|Config|GetbShouldDropOnEmpty))
      (Inventory|DropItembyGuid (Fortnite|GetFortPlayerControllerfromActor (Variables|Default|GetOwnerPawn)) (FortWeapon|GetInventoryGUID (Variables|/Script/FortniteGame.FortPawn|GetCurrentWeapon (Variables|Default|GetOwnerPawn)))))
    (else
      (CallFunction|ParentDoonEndAbility))))

(event Custom|Parent_DoOnEndAbility
  ; On End Ability
  (Ability|RemoveGameplayEffectFromOwnerWithHandle (Variables|Default|GetPassiveGEHandle))
  (CallFunction|RemoveBeamGC)
  (Utilities|IsValid (Variables|Default|GetTargetedBeamComponent)
    (:Is%20Valid
      (Beam|StopFiringBeam (Variables|Default|GetTargetedBeamComponent)))
    (:Is%20Not%20Valid
      (Fortnite|FortShippingLog self "Cached targeted beam component is invalid! Cannot call StopFiringBeam." true))))

(event Custom|StartSpendingAmmo
  (params (Rate :type Float%20%28double-precision%29) (ConsumeInstantly? :type Boolean))
  (Utilities|Time|ClearandInvalidateTimerbyHandle (Variables|Default|GetAbilityCostTimer))
  (bind _returnvalue (Utilities|Time|SetTimerbyEvent (EventDispatchers|CreateEvent :!event Custom|ConsumeAmmo) Rate true true))
  (Variables|Default|SetAbilityCostTimer _returnvalue)
  (Variables|/Sprites_WhisperRemedySweetSprite/Gameplay/ItemVersion/B_Sprite_Held_Weapon_WhisperRemedySweet.B_Sprite_Held_Weapon_WhisperRemedySweet_C|SetCurrentLaserState "NewEnumerator0" (Variables|Default|GetSourceWeapon))
  (if ConsumeInstantly?
    (CallFunction|ConsumeAmmo)))

(event Custom|ConsumeAmmo
  (if (<= (FortWeapon|GetMagazineAmmoCount (Variables|/Script/FortniteGame.FortPawn|GetCurrentWeapon (Variables|Default|GetOwnerPawn))) 1)
    (bind _returnvalue (FortWeapon|TryReloadAmmofromAbility (Variables|/Script/FortniteGame.FortPawn|GetCurrentWeapon (Variables|Default|GetOwnerPawn))))
    (Variables|Default|SetbIsEmpty (not _returnvalue))
    (Development|PrintString "MAG EMPTY !!!")
    (FortWeapon|Ammo|ConsumeAmmo (Variables|/Script/FortniteGame.FortPawn|GetCurrentWeapon (Variables|Default|GetOwnerPawn)) (ScalableFloat|GetValueasInteger (Variables|Config|GetAmmoConsumedPerTick)) false)
    (Ability|EndAbility)
    (else
      (FortWeapon|Ammo|ConsumeAmmo (Variables|/Script/FortniteGame.FortPawn|GetCurrentWeapon (Variables|Default|GetOwnerPawn)) (ScalableFloat|GetValueasInteger (Variables|Config|GetAmmoConsumedPerTick)) false))))

(event Custom|StartBeam
  ; Important to call StartFiringBeam before the Gameplay Cue runs any logic so it can use correct values of beam start/end points
  (if (ScalableFloat|GetValueasBool (Variables|Config|GetShouldWaitNetSync))
    (Ability|Tasks|WaitNetSync "OnlyClientWait"
      (:OnSync
        _)))
  (Beam|StartFiringBeam (Variables|Default|GetTargetedBeamComponent))
  (Ability|AddGameplayCueToOwner (Variables|Default|GetGCTag_Beam) (Class|AthenaFunctionLibrary|MakeSimpleFortEffectContext (Variables|Default|GetOwnerPawn) (Variables|Default|GetOwnerPawn)))
  ; Set timer to end the beam
  (bind _returnvalue (Utilities|Time|SetTimerbyEvent (EventDispatchers|CreateEvent :!event K2_EndAbility) (Beam|GetBeamDuration (Variables|Default|GetTargetedBeamComponent))))
  (Variables|Default|SetEndBeamTimerHandle _returnvalue)
  (Ability|HasAuthority
    (:True
      _)
    (:False
      (if (ScalableFloat|GetValueasBool (Variables|Config|GetOnlyConsumeAmmoOnServer))
        _
        (else
          _))))
  (CallFunction|StartSpendingAmmo (ScalableFloat|GetValueatLevel (Variables|Config|GetAmmoSpendRate)) (ScalableFloat|GetValueasBool (Variables|Config|GetShouldConsumeAmmoInstantly))))

(event Custom|RemoveBeamGC
  (bind _gctag_beam (Variables|Default|GetGCTag_Beam))
  (bind _returnvalue (Ability|GameplayCue|IsGameplayCueActive (Variables|Default|GetOwnerPawn) _gctag_beam))
  (if _returnvalue
    (Utilities|IsServer
      (:true
        (Ability|RemoveGameplayCueFromOwner _gctag_beam))
      (:false
        ; Client shouldn't have to wait on server to remove the gameplay cue (which it would have to otherwise because it's not authoritative and does not have a valid prediction key after the latent nodes)
        (Projectile|RemoveGameplayCue%28Looping%29Local (CallFunction|GetAvatarasFortPlayerPawn) (Variables|Default|GetGCTag_Beam) (Ability|GameplayCue|MakeGameplayCueParameters 0.0))))))

(event Ability|EventOnAbilityInputPressed
  (bind _returnvalue (CallFunction|GetAvatarasFortPlayerPawn))
  (|CallParentFunction)
  (bind _returnvalue_2 (Ability|CheckAbilityCooldown))
  (if (not _returnvalue_2)
    (Networking|CanExecuteCosmeticEvents
      (:True
        (Utilities|IsValid _returnvalue
          (:Is%20Valid
            (if (Pawn|IsLocallyViewed _returnvalue)
              (Audio|PlaySound2D "/Game/Athena/Sounds/Items/Ability_CantUse_Cue.Ability_CantUse_Cue"))))))))
