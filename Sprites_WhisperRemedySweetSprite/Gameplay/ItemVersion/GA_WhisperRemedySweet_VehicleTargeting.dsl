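; GA_WhisperRemedySweet_VehicleTargeting is a gameplay ability for the item version of the WhisperRemedySweet sprite that switches the camera for vehicle targeting.
; On activation it pushes the CameraMode_WhisperRemedySweet_VehicleTargeting camera mode.
; When the ability ends, whether cancelled or not, it pops that camera mode; it contains no other targeting logic and does not end itself.
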(event Ability|EventActivateAbility
  (Camera|PushCameraMode "/Sprites_WhisperRemedySweetSprite/Gameplay/ItemVersion/CameraMode_WhisperRemedySweet_VehicleTargeting.CameraMode_WhisperRemedySweet_VehicleTargeting_C"))

(event Ability|EventOnEndAbility
  (params (bWasCancelled :type Boolean))
  (Camera|PopCameraMode))
